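; No Sweat Summer quest sign for Battle Pass S21, a held object actor that players carry.
; After spawning, it applies the on-quest effect to its owner, is picked up server-side and rotates to fit the hands on attach.
; Only quest-tagged pawns can pick it up; releasing it realigns it and shows its minimap icon, with placement pushing it off overlapping players.
; Destruction calls its spawner, and leaving the POI triggers a client-side pop effect.
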(event EventBeginPlay
  (bind _fortheldobject (Variables|Default|GetFortHeldObject))
  ;[1 1 1] Begin Play
  (|CallParentFunction)
  (Utilities|FlowControl|Delay 0.5)
  (bind _asfort_player_pawn (Utilities|Casting|CastToFortPlayerPawn (Actor|GetOwner)))
  (Creative|GameplayEffects|SimpleApplyGameplayEffectToActor "/BattlePassS21/Items/QuestAbilities/Quest_NoSweatCarrySign/GE_Quest_NoSweat_IsOnQuest.GE_Quest_NoSweat_IsOnQuest_C" _asfort_player_pawn)
  (Variables|/Script/Engine.MovementComponent|SetVelocity "0, 0, 0" (Variables|Default|GetFortProjectileMovement))
  (if (Networking|IsServer)
    (HeldObjects|Behavior|PickupHeldObject _fortheldobject _asfort_player_pawn)
    ;[1 1 1] This rotates the sign to touch the characters hands
    (EventDispatchers|BindEventtoOnHeldObjectAttached _fortheldobject (delegate Custom|Rotate%20Sign))))

(event Custom|Rotate%20Sign
  ;[1 1 1] This rotates the sign to touch the characters hands
  (Transformation|SetActorRelativeRotation "-2.000000,180.000000,-90.000000")
  (CallFunction|ControlIcons false))

(event OnHeldObjectPlaced(FortHeldObject)
  (bind (_returnvalue_roll _returnvalue_pitch _returnvalue_yaw) (Math|Rotator|BreakRotator (Transformation|GetActorRotation (HeldObjects|Instigator|GetInstigator (Variables|Default|GetFortHeldObject)))))
  ;[1 1 1] When we set the sign down, if it is overlapping with a player, push the sign away from the direction the player is facing
  (Utilities|Array|ForEachLoopwithBreak (Collision|GetOverlappingActors (Variables|Default|GetCapsule) "/Script/FortniteGame.FortPlayerPawn" :!class /Script/Engine.PrimitiveComponent)
    (:LoopBody
      (Variables|/Script/Engine.MovementComponent|SetVelocity (* (+ (* (Math|Vector|GetForwardVector (Transformation|GetActorRotation (HeldObjects|Instigator|GetInstigator (Variables|Default|GetFortHeldObject)))) -1.0) "0, 0,6.000000") 2e+02) (Variables|Default|GetFortProjectileMovement)))
    (:Completed
      _))
  ;[1 1 1] Set variables to track player interaction; report quest progress
  (Variables|Default|SetHeld false)
  (Transformation|SetActorRotation 0 (Math|Rotator|MakeRotator 0.0 0.0 (+ _returnvalue_yaw 9e+01)) false)
  (CallFunction|ControlIcons true))

(event Interaction|EventBlueprintOnInteract
  (params (InteractingPawn :type Fort%20Pawn%20Object%20Reference) (InteractionBeingAttempted :type EInteractionBeingAttempted%20Enum :default "FirstInteraction"))
  ;[1 1 1] On Interact
  (bind _bdidactorhavetags (Class|AthenaQuestFunctionLibrary|AthenaCheckActorHasTags InteractingPawn "(GameplayTags=((TagName=\"Athena.Quests.S21.SignOnQuest\")))" true))
  (if _bdidactorhavetags
    (switch Utilities|FlowControl|Switch|SwitchonEInteractionBeingAttempted InteractionBeingAttempted
      (:FirstInteraction
        (bind _asfort_player_pawn (Utilities|Casting|CastToFortPlayerPawn InteractingPawn))
        (Variables|/Script/Engine.MovementComponent|SetVelocity "0, 0, 0" (Variables|Default|GetFortProjectileMovement))
        (if (Networking|IsServer)
          (HeldObjects|Behavior|PickupHeldObject (Variables|Default|GetFortHeldObject) _asfort_player_pawn)
          (CallFunction|ControlIcons false))))))

(event OnHeldObjectDropped(FortHeldObject)
  (bind (_returnvalue_roll _returnvalue_pitch _returnvalue_yaw) (Math|Rotator|BreakRotator (Transformation|GetActorRotation (HeldObjects|Instigator|GetInstigator (Variables|Default|GetFortHeldObject)))))
  ;[1 1 1] Set variables to track player interaction; report quest progress
  (Variables|Default|SetHeld false)
  (Transformation|SetActorRotation 0 (Math|Rotator|MakeRotator 0.0 0.0 (+ _returnvalue_yaw 9e+01)) false)
  (CallFunction|ControlIcons true))

(event OnHeldObjectThrown(FortHeldObject)
  (bind (_returnvalue_roll _returnvalue_pitch _returnvalue_yaw) (Math|Rotator|BreakRotator (Transformation|GetActorRotation (HeldObjects|Instigator|GetInstigator (Variables|Default|GetFortHeldObject)))))
  ;[1 1 1] Set variables to track player interaction; report quest progress
  (Variables|Default|SetHeld false)
  (Transformation|SetActorRotation 0 (Math|Rotator|MakeRotator 0.0 0.0 (+ _returnvalue_yaw 9e+01)) false)
  (CallFunction|ControlIcons true))

(event Custom|Control%20Icons
  (params (ShowIcons :type Boolean))
  (CallFunction|OnRepShowIcon ShowIcons))

(event OnLinkedActorDisconnected(FortLinkToActor)
  (if (not (Variables|Default|GetHeld))
    (Variables|ProjectileMovement|SetRepStartMoving 1)))

(event EventDestroyed
  (Class|BActorSpawnerParent|SpawnActor :self (Variables|Default|GetMySpawner)))

(event Custom|ClientEnd
  (if (not (Networking|IsServer))
    (Components|Activation|Activate (Variables|Default|GetDeathExplosion) false)
    (Development|SetHiddeninGame (Variables|Default|GetStaticMesh) true false)))

(fn BlueprintCanInteract
  (params (InteractingPawn :type Fort%20Pawn%20Object%20Reference) (InteractionBeingAttempted :type EInteractionBeingAttempted%20Enum :default "FirstInteraction") (InteractionType :type TInteractionType%20Enum :default "IT_NoInteraction"))
  (Utilities|IsValid InteractingPawn
    (:Is%20Valid
      (bind _returnvalue (HeldObjects|Interaction|CanInteract (Variables|Default|GetFortHeldObject) (Utilities|Casting|CastToFortPlayerPawn InteractingPawn)))
      (return (Math|Boolean|ANDBoolean _returnvalue (not (GameplayTags|HasMatchingGameplayTag InteractingPawn "(TagName=\"Gameplay.Status.HidingInProp\")")) (not (GameplayTags|HasAnyMatchingGameplayTags InteractingPawn (Variables|Interact|GetTC_BlockPickup))) (GameplayTags|HasMatchingGameplayTag InteractingPawn "(TagName=\"Athena.Quests.S21.SignOnQuest\")"))))))

(fn BlueprintGetInteractionString
  (params (InteractingPawn :type Fort%20Pawn%20Object%20Reference) (InteractionBeingAttempted :type EInteractionBeingAttempted%20Enum :default "FirstInteraction"))
  (return (Variables|Default|GetInteractString)))

(fn OnRep_bIsInPOI
  (if (not (Variables|Default|GetbIsInPOI))
    (CallFunction|ClientEnd)))

(fn OnRep_Held
  (if (Variables|Default|GetHeld)
    (Utilities|Time|SetTimerbyFunctionName 0 "POI Overlap Check" 0.1 true)
    (else
      (Utilities|Time|ClearTimerbyFunctionName 0 "POI Overlap Check"))))

(fn OnRep_ShowIcon
  (params (ShowIcons :type Boolean))
  (bind _fortquesticon (Variables|Default|GetFortQuestIcon))
  (MiniMapData|SetViewDistanceRadius _fortquesticon (MiniMapData|GetViewDistanceRadius _fortquesticon))
  (MiniMapData|SetVisibility _fortquesticon ShowIcons))

(fn SetAttachDefaultTransform)

(fn ShouldSignBeEnabled
  (bind _outxy (DataTable|EvaluateCurveTableRow "/Game/Athena/Balance/DataTables/AthenaGameData.AthenaGameData" "Default.QuestInteractable.GiantPinkBear.Enabled")
    (:RowFound
      (if (>= _outxy 1.0)
        (return true)
        (else
          (return false))))
    (:RowNotFound
      (return false))))

(fn ConstructionScript
  (|CallParentFunction)
  (Effects|Components|ParticleSystem|SetTemplate (Variables|Default|GetDeathExplosion) (Variables|Default|GetBearPopEffect)))
