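; Season 21 No Sweat Summer quest actor that reacts when a player carrying a held object enters its hidden collision.
; On overlap with SM_HiddenInteractionCollision, it casts to FortPlayerPawn and gets the pawn's attached held object.
; If that object's class is valid, with no specific type check, it applies the CarrySign quest GameplayEffect to the pawn and destroys the object.
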(event EventBeginPlay
  (|CallParentFunction))

(event EventTick
  (params (DeltaSeconds :type Float%20%28single-precision%29))
  (|CallParentFunction DeltaSeconds) ; This node is disabled and will not be called.\nDrag off pins to build functionality.
)

(event OnComponentBeginOverlap(SM_HiddenInteractionCollision)
  (params (OverlappedComponent :type Primitive%20Component%20Object%20Reference) (OtherActor :type Actor%20Object%20Reference) (OtherComp :type Primitive%20Component%20Object%20Reference) (OtherBodyIndex :type Integer) (bFromSweep :type Boolean) (SweepResult :type Hit%20Result%20Structure%20%28by%20ref%29))
  (bind _asfort_player_pawn (Utilities|Casting|CastToFortPlayerPawn OtherActor))
  (bind _returnvalue (Fortnite|HeldObjects|GetHeldObjectAttachedtoPlayer _asfort_player_pawn))
  (if (Utilities|IsValidClass (Utilities|GetClass _returnvalue))
    (Creative|GameplayEffects|SimpleApplyGameplayEffectToActor "/BattlePassS21/Items/QuestAbilities/Quest_NoSweatCarrySign/GE_Quest_NoSweat_CarrySign.GE_Quest_NoSweat_CarrySign_C" _asfort_player_pawn)
    (Actor|DestroyActor _returnvalue)))

(fn GetBoundsComponents
  (params (SelfActor :type Actor%20Object%20Reference))
  (return 0))

(fn ConstructionScript
  (|CallParentFunction))
