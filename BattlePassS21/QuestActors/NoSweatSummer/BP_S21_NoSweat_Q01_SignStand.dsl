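; Season 21 No Sweat Summer quest sign stand that spawns a held-object quest sign when a pawn interacts with it.
; The interact event builds the spawn transform from the stand's X and Y with fixed world-space offsets, leaving Z unwired, and tilts its rotation.
; The sign spawns with the interacting pawn passed in, with no authority or repeat-interaction check in this graph.
; BeginPlay, overlap and Tick are disabled stubs.
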(event EventBeginPlay
  (|CallParentFunction) ; This node is disabled and will not be called.\nDrag off pins to build functionality.
)

(event Collision|EventActorBeginOverlap
  (params (OtherActor :type Actor%20Object%20Reference))
  (|CallParentFunction OtherActor) ; This node is disabled and will not be called.\nDrag off pins to build functionality.
)

(event EventTick
  (params (DeltaSeconds :type Float%20%28single-precision%29))
  (|CallParentFunction DeltaSeconds) ; This node is disabled and will not be called.\nDrag off pins to build functionality.
)

(event Interaction|EventBlueprintOnInteract
  (params (InteractingPawn :type Fort%20Pawn%20Object%20Reference) (InteractionBeingAttempted :type EInteractionBeingAttempted%20Enum :default "FirstInteraction"))
  (bind (_returnvalue_x _returnvalue_y) (Math|Vector|BreakVector (Transformation|GetActorLocation)))
  (bind (_returnvalue_roll _returnvalue_pitch _returnvalue_yaw) (Math|Rotator|BreakRotator (Transformation|GetActorRotation)))
  ;[1 1 1] Spawn Sign in stand at base
  (Game|SpawnActorfromClass "/BattlePassS21/QuestActors/NoSweatSummer/BGA_HeldObject_Quest_NoSweatSIgn.BGA_HeldObject_Quest_NoSweatSIgn_C" (Math|Transform|MakeTransform (Math|Vector|MakeVector (- _returnvalue_x -7.0) (- _returnvalue_y -14.0)) (Math|Rotator|MakeRotator (- _returnvalue_roll 9e+01) _returnvalue_pitch (- _returnvalue_yaw 7.0))) "Undefined" "OverrideRootScale" InteractingPawn 0 0))

(fn ConstructionScript
  (|CallParentFunction))
