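; Cosmetic sign stand actor for the No Sweat Summer Q01 quest that animates the sign's removal, then destroys itself.
; On BeginPlay it stores the mesh's initial position and rotation and calls an attached sound spawn on the stand.
; SignRemovedTimeline then drives the mesh's relative height, rotation and scale each update, without applying the stored initial values.
; The actor is destroyed when the timeline finishes.
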(event EventBeginPlay
  (bind _sm_signstand (Variables|Default|GetSM_SignStand))
  (Variables|Default|SetInitialPosition (Variables|/Script/Engine.SceneComponent|GetRelativeLocation _sm_signstand))
  (Variables|Default|SetInitialRotation (Variables|/Script/Engine.SceneComponent|GetRelativeRotation _sm_signstand))
  (Audio|SpawnSoundAttached 0 (Variables|Default|GetSM_SignStand))
  (bind (_direction _mortarshellsheight _mortarshellsrotation _mortarshellsscale) (Timeline|SignRemovedTimeline PlayFromStart)
    (:Update
      (Transformation|SetRelativeTransform (Variables|Default|GetSM_SignStand) (Math|Transform|MakeTransform (Math|Vector|MakeVector 0.0 0.0 _mortarshellsheight) (Math|Rotator|MakeRotator (.x _mortarshellsrotation) (.y _mortarshellsrotation) (.z _mortarshellsrotation)) _mortarshellsscale) false false))
    (:Finished
      (Actor|DestroyActor))))

(fn ConstructionScript)
