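; Actor spawner for the No Sweat Summer quest, which by its name places held No Sweat sign objects in the world.
; On its spawn event it checks the class from GetClassToSpawn and always spawns it at the spawner's transform.
; A spawned No Sweat sign receives the spawner's POI list as location tags and has its MySpawner variable set.
; If that cast fails, the result is cast to BuildingActor and initialized as a kismet-spawned building actor.
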(event EventBeginPlay
  (|CallParentFunction) ; This node is disabled and will not be called.\nDrag off pins to build functionality.
)

(event Collision|EventActorBeginOverlap
  (params (OtherActor :type Actor%20Object%20Reference)))

(event EventTick
  (params (DeltaSeconds :type Float%20%28single-precision%29)))

(event EventSpawnActor
  (params (New%20Transform :type Transform))
  (bind _classtospawn (Variables|SpawnerVars|GetClassToSpawn))
  ;[1 1 1] If the Actor is a valid class spawn it. If it is a building actor init the kismet of it.
  (if (Utilities|IsValidClass _classtospawn)
    (bind _returnvalue (Game|SpawnActorfromClass _classtospawn (Transformation|GetActorTransform) "AlwaysSpawn" "SelectDefaultAtRuntime" 0 0))
    (bind _asbga_held_object_quest_no_sweat_sign (Utilities|Casting|CastToBGA_HeldObject_Quest_NoSweatSIgn _returnvalue)
      (:then
        (Variables|/BattlePassS21/QuestActors/NoSweatSummer/BGA_HeldObject_Quest_NoSweatSIgn.BGA_HeldObject_Quest_NoSweatSIgn_C|SetLocationTags (Variables|Default|GetPOI%20List) _asbga_held_object_quest_no_sweat_sign)
        (Variables|/BattlePassS21/QuestActors/NoSweatSummer/BGA_HeldObject_Quest_NoSweatSIgn.BGA_HeldObject_Quest_NoSweatSIgn_C|SetMySpawner 0 _asbga_held_object_quest_no_sweat_sign))
      (:CastFailed
        (bind _asbuilding_actor (Utilities|Casting|CastToBuildingActor _returnvalue))
        (Building|InitializeKismetSpawnedBuildingActor _asbuilding_actor 0 0 false 0 false)))))

(fn ConstructionScript
  (|CallParentFunction))
