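# Masked material for Niagara mesh particles in the No Sweat Summer quest, showing signs burning in a trash-can barrel.
# The second UV channel selects sign or barrel texturing, masking and emissive.
# Signs erode through a radial gradient broken by screen-aligned noise, with glowing edges and fake firelight.
# The barrel gets normal-based inner light and height-masked, dithered soot erosion; dynamic parameters drive erosion and lighting.
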Material /BattlePassS21/QuestActors/NoSweatSummer/FX/Materials/M_Quest_NoSweat_BurnBarrel
graph {
 properties {
  bFullyRough=true
  bUsedWithNiagaraMeshParticles=true
  blendMode=BLEND_Masked
  pixelDepthOffsetMode=PDOM_Legacy
  usedShadingModels=Default Lit
 }
 nodes [
  n1 = Floor(None=ComponentMask(None=TextureCoordinate() {coordinateIndex=1}) {r=true})
  LinearInterpolate(
   A=TextureSample() {texture=/Game/Environments/Artemis/Props/Quest/Quest_SF_TrashCan_A/Textures/T_Quest_SF_WallSign_A_D}.RGB
   B=TextureSample() {texture=/Game/Environments/Artemis/Props/Quest/Quest_SF_TrashCan_A/Textures/T_Quest_SF_TrashCan_A_D}.RGB
   Alpha=@n1
  )
  n2 = MaterialFunctionCall() {materialFunction=&ObjectPivotPoint}
  n3 = VectorParameter() {parameterName=Burn Center, defaultValue=<rgba 0 0 33 0>}
  n4 = Divide(
   A=Subtract(A=WorldPosition().XYZ, B=Add(A=@n2.Mesh Particle Pivot Location, B=@n3.RGB))
   B=VectorParameter() {parameterName=Burn Radius, defaultValue=<rgba 66 44 222 0>}.RGB
  )
  n5 = Saturate(None=MaterialFunctionCall(Vector 2=ComponentMask(None=@n4) {b=true, r=true}) {materialFunction=/Engine/Functions/Engine_MaterialFunctions02/Utility/VectorLength}.V2 Length)
  n6 = MaterialFunctionCall!(Center Point World Position) {materialFunction=/Engine/Functions/Engine_MaterialFunctions02/UVs/WorldSpaceAlignedScreenCoordinates}
  n7 = TextureSample!(
   UVs=Multiply(A=ConstantBiasScale(None=@n6.Offset and Scaled UVS), B=ScalarParameter() {defaultValue=22, parameterName=Noise UV Scale})
   Tex
   Level=ScalarParameter() {defaultValue=1, parameterName=Noise Mip Level}
   Apply View MipBias
  ) {mipValueMode=TMVM_MipLevel, texture=/Game/Effects/Fort_Effects/Textures/Fire/T_FireNoise_02}
  n8 = ScalarParameter() {defaultValue=2, parameterName=Noise Exponent}
  n9 = Multiply(A=Subtract(A=Power(Base=@n7.R, Exp=@n8)) {constB=0.5}, B=ScalarParameter() {defaultValue=0.666, parameterName=Noise Intensity})
  n10 = DynamicParameter() {defaultValue=<rgba 0.5 4 1 1>, paramNames [Burning Edge Erosion, Signs Light Exponent, Barrel Light Intensity, Barrel Soot Erosion]}
  n11 = MaterialFunctionCall!(
   Gradient=MaterialFunctionCall(Base=@n5, Blend=ConstantBiasScale(None=@n9)) {materialFunction=&Blend_Overlay}.Result
   Erosion Amount=@n10.Burning Edge Erosion
  ) {materialFunction=/Game/Effects/Fort_Effects/MaterialFunctions/MF_ErosionNormalized}
  n12 = DynamicParameter() {defaultValue=<rgba 0.5 4 1 1>, paramNames [Burning Edge Erosion, Signs Light Exponent, Barrel Light Intensity, Barrel Soot Erosion]}
  n13 = ComponentMask(None=MaterialFunctionCall() {materialFunction=&LocalPosition}.Local Position) {b=true}
  n14 = ScalarParameter() {defaultValue=22, parameterName=Base Mask Offset}
  n15 = Saturate(None=Divide(A=Subtract(A=@n13, B=@n14), B=ScalarParameter() {defaultValue=33, parameterName=Base Mask Distance}))
  LinearInterpolate(
   A=Ceil(None=@n11.Result)
   B=MaterialFunctionCall(Alpha Threshold=Multiply(A=@n12.Barrel Soot Erosion, B=@n15)) {materialFunction=&DitherTemporalAA_Fortnite}.Result
   Alpha=@n1
  )
  n16 = NamedRerouteUsage()
  n17 = OneMinus(None=@n1)
  n18 = Divide(A=@n11.Result, B=ScalarParameter() {defaultValue=0.25, parameterName=Burning Edge Size})
  n19 = Divide(A=Frac(None=Saturate(None=@n18)))
  n20 = OneMinus(None=Saturate(None=Subtract(A=@n5, B=@n10.Burning Edge Erosion)))
  n21 = Add(
   A=Multiply(A=Sine(None=@n19), B=VectorParameter() {parameterName=Burning Edge, defaultValue=<rgba 22 2.992 0.88 0>}.RGB)
   B=Multiply(A=Power(Base=@n20, Exp=@n10.Signs Light Exponent), B=VectorParameter() {parameterName=Fake Light, defaultValue=<rgba 3 0.456 0.12 4>}.RGB) {constA=1}
  )
  MaterialFunctionCall(Emissive=Add(A=Multiply(A=@n1, B=@n16), B=Multiply(A=@n17, B=@n21))) {materialFunction=&MF_FX_EmissiveAdjustments}
  n22 = CameraPositionWS()
  n23 = WorldPosition()
  Multiply(A=Normalize(VectorInput=Subtract(A=@n22, B=@n23.XYZ)), B=ScalarParameter() {defaultValue=3, parameterName=Camera Offset})
  n24 = Constant3Vector() {constant=<rgba 0 0 1 0>}
  n25 = ScalarParameter() {defaultValue=111, parameterName=Barrel Inner Light Center Offset}
  n26 = WorldPosition()
  n27 = Subtract(
   A=Add(
    A=MaterialFunctionCall() {materialFunction=&ObjectPivotPoint}.Mesh Particle Pivot Location
    B=Multiply(A=@n24, B=@n25)
   )
   B=@n26.XYZ
  )
  n28 = Subtract(A=@n26.XYZ, B=ActorPositionWS())
  n29 = Divide(
   A=MaterialFunctionCall(Vector 2=ComponentMask(None=@n28) {g=true, r=true}) {materialFunction=&VectorLength}.V2 Length
   B=ScalarParameter() {defaultValue=44, parameterName=Barrel Inner Light Radius}
  )
  n30 = Multiply(A=DotProduct(A=Normalize(VectorInput=@n27), B=VertexNormalWS()), B=OneMinus(None=Saturate(None=@n29)))
  n31 = VectorParameter() {parameterName=Barrel Light, defaultValue=<rgba 3 0.456 0.12 4>}
  n32 = DynamicParameter() {defaultValue=<rgba 0.5 4 1 1>, paramNames [Burning Edge Erosion, Signs Light Exponent, Barrel Light Intensity, Barrel Soot Erosion]}
  NamedRerouteDeclaration(None=Multiply(A=Multiply(A=@n30, B=@n31.RGB), B=@n32.Barrel Light Intensity)) {name=Barrel Light, nodeColor=<rgba 1 0.6094 0 1>}
  n33 = Subtract(
   A=WorldPosition().XYZ
   B=MaterialFunctionCall() {materialFunction=&ObjectPivotPoint}.Mesh Particle Pivot Location
  )
  n34 = Saturate(None=Divide(A=ComponentMask(None=@n33) {b=true}) {constB=167})
  n35 = ScalarParameter() {defaultValue=0.333, parameterName=Gradient Center Point}
  VertexInterpolator(VS=OneMinus(None=MaterialFunctionCall!(Gradient=@n34, Center Point=@n35) {materialFunction=/Game/Effects/Fort_Effects/MaterialFunctions/MF_GradientReflected}.Result))
 ]
}
